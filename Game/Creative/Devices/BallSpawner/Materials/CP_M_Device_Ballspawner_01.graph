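# Default Lit material for the Creative Ball Spawner device's ball, with adjustable tint, logo brightness, roughness and metalness.
# Base color multiplies the diffuse texture by ColorTint and remapped Brightness, then adds the alpha channel scaled by LogoBrightness.
# A packed mask supplies specular, offsets metalness and scales remapped roughness, alongside a normal map.
# All tunables except Brightness use custom primitive data for the edit-only mesh; attributes pass through MF_CreativeMoveTool.
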Material /Game/Creative/Devices/BallSpawner/Materials/CP_M_Device_Ballspawner_01
graph {
 properties {bUseMaterialAttributes=true, pixelDepthOffsetMode=PDOM_Legacy, usedShadingModels=Default Lit}
 nodes [
  n1 = VectorParameter() {
   bUseCustomPrimitiveData=true
   desc=The edit only mesh uses primitive data so we need to use primitive data
   parameterName=ColorTint
   defaultValue=<rgba 0 0.2502 0.6724 1>
  }
  n2 = TextureSample() {texture=/Game/Creative/Devices/BallSpawner/Textures/T_CP_Device_BallSpawner_Ball_01_D}
  n3 = Multiply(
   A=MaterialFunctionCall(
    Input=ScalarParameter() {defaultValue=0.5, parameterName=Brightness, sliderMax=1}
    Input Low=Constant()
    Input High=Constant() {r=1}
    Target Low=Constant()
    Target High=Constant() {r=2}
   ) {materialFunction=&RemapValueRange}.Result
   B=Multiply(A=@n1.RGB, B=@n2.RGB)
  )
  n4 = Multiply(
   A=ScalarParameter() {bUseCustomPrimitiveData=true, defaultValue=0.2, parameterName=LogoBrightness, primitiveDataIndex=6, sliderMax=1}
   B=@n2.A
  )
  n5 = TextureSample() {samplerType=SAMPLERTYPE_Masks, texture=/Game/Creative/Devices/BallSpawner/Textures/T_CP_Device_BallSpawner_Ball_01_S}
  n6 = MaterialFunctionCall(
   Input=ScalarParameter() {bUseCustomPrimitiveData=true, defaultValue=0.4, parameterName=Roughness, primitiveDataIndex=4, sliderMax=1}
   Input Low=Constant()
   Input High=Constant() {r=1}
   Target Low=Constant()
   Target High=Constant() {r=2.5}
  ) {materialFunction=&RemapValueRange}
  n7 = ScalarParameter() {bUseCustomPrimitiveData=true, parameterName=Metalness, primitiveDataIndex=5, sliderMax=1}
  MaterialFunctionCall(MA In=MakeMaterialAttributes!(
    BaseColor=Add(A=@n3, B=@n4)
    Metallic=Add(A=@n7, B=@n5.G)
    Specular=@n5.R
    Roughness=Multiply(A=@n6.Result, B=@n5.B)
    Anisotropy
    EmissiveColor
    Opacity
    OpacityMask
    Normal=TextureSample() {samplerType=SAMPLERTYPE_Normal, texture=/Game/Creative/Devices/BallSpawner/Textures/T_CP_Device_BallSpawner_Ball_01_N}.RGB
    Tangent
    WorldPositionOffset
    SubsurfaceColor
    ClearCoat
    ClearCoatRoughness
    AmbientOcclusion
    Refraction
    CustomizedUVs_0
    CustomizedUVs_1
    CustomizedUVs_2
    CustomizedUVs_3
    CustomizedUVs_4
    CustomizedUVs_5
    CustomizedUVs_6
    CustomizedUVs_7
    PixelDepthOffset
    ShadingModel
    Displacement
   )) {materialFunction=&MF_CreativeMoveTool}
 ]
}
